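# Conversation database for the Summer Drift NPC, defining the dialogue opened by the Conversation.Entry.SummerDrift entry tag.
# The entry leads to one speech node with a default surfing greeting and contextual lines for a first-ever conversation or specific Drift and fox-themed character outfits.
# Contextual lines pick by requirement and priority, with the first-meeting priority read from a curve row.
# The speech then links to the shared NPC library Quests and BasicServices conversations.
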ConversationDatabase /NPCLibrary/NPCs/SummerDrift/NPCLibrary_Conversation_SummerDrift
entry_points [{tag=Conversation.Entry.SummerDrift, destinations [#1]}]
nodes [
 ConversationEntryPointNode_7 {
  id=#1
  class=ConversationEntryPointNode
  outputs [#2]
  properties {entryTag {tagName=Conversation.Entry.SummerDrift}}
 }
 Speech_C_4 {
  id=#2
  class=Speech_C
  outputs [#3, #4]
  properties {
   nodeName=""
   generalConfig {
    defaultMessage=Any good waves out there today?
    contextualMessages [
     {
      message="Beautiful day, today, my friend! Hope you're well."
      contextRequirements [{requirement {class=FortControllerRequirement_IsFirstEverConversationWithNPC}}]
      priority {curve {rowName=Tandem.Speech.Priority.FirstEverInteraction}}
     }
     {
      message=Are you here on Fox Clan business?
      requirementMatchPolicy=RequireAny
      contextRequirements [{requirement {
         class=FortControllerRequirement_HasCID
         properties {allowedCIds [
           /BRCosmetics/Athena/Items/Cosmetics/Characters/CID_161_Athena_Commando_M_Drift
           /BRCosmetics/Athena/Items/Cosmetics/Characters/CID_461_Athena_Commando_M_DriftSummer
           /BRCosmetics/Athena/Items/Cosmetics/Characters/CID_982_Athena_Commando_M_DriftWinter
           /BRCosmetics/Athena/Items/Cosmetics/Characters/CID_A_230_Athena_Commando_M_DriftHorror
           /BRCosmetics/Athena/Items/Cosmetics/Characters/CID_486_Athena_Commando_F_StreetRacerDrift
           /BRCosmetics/Athena/Items/Cosmetics/Characters/CID_A_009_Athena_Commando_F_FoxWarrior_21B9R
          ]}
        }}]
      priority {value=0, curve {curveTable=null, rowName=None}}
     }
    ]
   }
  }
 }
 ConversationLinkNode_0 {id=#4, class=ConversationLinkNode, properties {remoteEntryTag {tagName=Conversation.Entry.BasicServices}}}
 ConversationLinkNode_1 {id=#3, class=ConversationLinkNode, properties {remoteEntryTag {tagName=Conversation.Entry.NPCLibrary.Quests}}}
]
